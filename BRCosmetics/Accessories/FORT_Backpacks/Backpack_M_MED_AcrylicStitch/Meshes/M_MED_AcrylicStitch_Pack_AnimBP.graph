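# Animation Blueprint for the AcrylicStitch backpack cosmetic, adjusting the pack bone's placement and adding physics sway.
# Each update maps spine_05's yaw delta from ref pose, relative to spine_01, into the Parashute value.
# A constant additive offset on the pack bone feeds a second offset and roll weighted by Parashute.
# Three chained AnimDynamics nodes then simulate the Dyn_tree bones with angular springs, tight limits and damped component motion.
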AnimBlueprint /BRCosmetics/Accessories/FORT_Backpacks/Backpack_M_MED_AcrylicStitch/Meshes/M_MED_AcrylicStitch_Pack_AnimBP
target_skeleton=&/M_MED_AcrylicStitch_Pack_Skeleton
graphs [
 AnimGraph {nodes [
   n1 = "Transform (Modify) Bone - Bone: pack"(ComponentPose=Local To Component(LocalPose=Input Pose()), Translation="0,-1.8,-5", Rotation="0,0,0") {rotationMode=BMM_Additive, scaleMode=BMM_Additive, translationMode=BMM_Additive, boneToModify {boneName=pack}, translation {y=-1.8, z=-5}}
   n2 = "Transform (Modify) Bone - Bone: pack"(ComponentPose=@n1, Alpha=Get Parashute() -> (Parashute:double), Translation="0,-7,4", Rotation="0.000,0.000,20.000") {
    rotationMode=BMM_Additive
    scaleMode=BMM_Additive
    translationMode=BMM_Additive
    boneToModify {boneName=pack}
    rotation {roll=20}
    translation {y=-7, z=4}
   }
   n3 = Anim Dynamics - Bone: Dyn_tree_01(ComponentPose=@n2, Alpha=1, ExternalForce="0,0,0") {
    angularDampingOverride=0.9999
    angularSpringConstant=500
    bAngularSpring=true
    bOverrideAngularDamping=true
    bUseGravityOverride=true
    simulationSpace=RootRelative
    boundBone {boneName=Dyn_tree_01}
    componentAppliedLinearAccClamp=<600 600 600>
    componentLinearAccScale=<0.2 0.2 0.2>
    componentLinearVelScale=<0.2 0.2 0.2>
    gravityOverride {z=-980}
    physicsBodyDefinitions [{
      boundBone {boneName=Dyn_tree_01}
      boxExtents=<10 10 10>
      constraintSetup {angularLimitsMax=<0 5 3>, angularLimitsMin=<0 -5 -5>, angularTarget=<1 0 0>}
      localJointOffset=<0.8 0 0>
     }]
   }
   n4 = Anim Dynamics - Bone: Dyn_tree_02(ComponentPose=@n3, Alpha=1, ExternalForce="0,0,0") {
    angularDampingOverride=0.9999
    angularSpringConstant=500
    bAngularSpring=true
    bOverrideAngularDamping=true
    bUseGravityOverride=true
    simulationSpace=RootRelative
    boundBone {boneName=Dyn_tree_02}
    componentAppliedLinearAccClamp=<600 600 600>
    componentLinearAccScale=<0.2 0.2 0.2>
    componentLinearVelScale=<0.2 0.2 0.2>
    gravityOverride {z=-980}
    physicsBodyDefinitions [{
      boundBone {boneName=Dyn_tree_02}
      boxExtents=<10 10 10>
      constraintSetup {angularLimitsMax=<0 5 3>, angularLimitsMin=<0 -5 -5>, angularTarget=<1 0 0>}
      localJointOffset=<0.8 0 0>
     }]
   }
   n5 = Anim Dynamics - Bone: Dyn_tree_03(ComponentPose=@n4, Alpha=1, ExternalForce="0,0,0") {
    angularDampingOverride=0.9999
    angularSpringConstant=500
    bAngularSpring=true
    bOverrideAngularDamping=true
    bUseGravityOverride=true
    simulationSpace=RootRelative
    boundBone {boneName=Dyn_tree_03}
    componentAppliedLinearAccClamp=<600 600 600>
    componentLinearAccScale=<0.2 0.2 0.2>
    componentLinearVelScale=<0.2 0.2 0.2>
    gravityOverride {z=-980}
    physicsBodyDefinitions [{
      boundBone {boneName=Dyn_tree_03}
      boxExtents=<10 10 10>
      constraintSetup {angularLimitsMax=<0 5 3>, angularLimitsMin=<0 -5 -5>, angularTarget=<1 0 0>}
      localJointOffset=<0.8 0 0>
     }]
   }
   Output Pose(Result=Component To Local(ComponentPose=@n5)) {layerGroup=None}
  ]}
 EventGraph {nodes [
   Event BlueprintUpdateAnimation()
   n1 = GetDeltaTransformFromRefPose!(self=GetOwnerSkeletalMeshComponent(), BoneName=spine_05, BaseName=spine_01) -> (ReturnValue_Location:Vector, ReturnValue_Rotation:Rotator, ReturnValue_Scale:Vector)
   > Set Parashute(Parashute:double=MapRangeClamped(Value=BreakRotator(InRot=@n1.ReturnValue_Rotation).Yaw, InRangeA=-15, InRangeB=-50)) -> (Output_Get:double)
   TryGetPawnOwner()
  ]}
]
variables [Parashute {type=number}]
